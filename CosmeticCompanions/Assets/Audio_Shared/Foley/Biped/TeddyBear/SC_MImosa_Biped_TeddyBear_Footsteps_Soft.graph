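# Sound cue for soft footsteps of the TeddyBear biped cosmetic companion.
# A quality-level switch picks a ten-clip walk pool for its first two levels and a four-clip subset for the last two, each chosen at random with equal weight.
# A modulator varies pitch and volume per play.
# The cue plays at reduced volume and pitch, with companion foley attenuation, the pets sound class, and biped footstep concurrency.
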SoundCue /CosmeticCompanions/Assets/Audio_Shared/Foley/Biped/TeddyBear/SC_MImosa_Biped_TeddyBear_Footsteps_Soft
root=Modulator_0
nodes [
 {id=Modulator_0, children [QualityLevel_0], properties {pitchMax=1.2, pitchMin=0.9, volumeMax=1, volumeMin=0.8}}
 {id=QualityLevel_0, children [Random_2,Random_2,Random_3,Random_3]}
 {
  id=Random_2
  children [
   WavePlayer_28
   WavePlayer_29
   WavePlayer_30
   WavePlayer_31
   WavePlayer_32
   WavePlayer_33
   WavePlayer_34
   WavePlayer_35
   WavePlayer_36
   WavePlayer_37
  ]
  properties {weights [1,1,1,1,1,1,1,1,1,1]}
 }
 {id=Random_3, children [WavePlayer_38, WavePlayer_39, WavePlayer_40, WavePlayer_41], properties {weights [1,1,1,1]}}
 {
  id=WavePlayer_28
  properties {soundWaveAssetPtr=/CosmeticCompanions/Assets/Audio_Shared/Foley/Biped/TeddyBear/Wavs/Mimosa_Biped_TeddyBear_Walk_10}
 }
 {
  id=WavePlayer_29
  properties {soundWaveAssetPtr=/CosmeticCompanions/Assets/Audio_Shared/Foley/Biped/TeddyBear/Wavs/Mimosa_Biped_TeddyBear_Walk_09}
 }
 {
  id=WavePlayer_30
  properties {soundWaveAssetPtr=/CosmeticCompanions/Assets/Audio_Shared/Foley/Biped/TeddyBear/Wavs/Mimosa_Biped_TeddyBear_Walk_08}
 }
 {
  id=WavePlayer_31
  properties {soundWaveAssetPtr=/CosmeticCompanions/Assets/Audio_Shared/Foley/Biped/TeddyBear/Wavs/Mimosa_Biped_TeddyBear_Walk_07}
 }
 {
  id=WavePlayer_32
  properties {soundWaveAssetPtr=/CosmeticCompanions/Assets/Audio_Shared/Foley/Biped/TeddyBear/Wavs/Mimosa_Biped_TeddyBear_Walk_06}
 }
 {
  id=WavePlayer_33
  properties {soundWaveAssetPtr=/CosmeticCompanions/Assets/Audio_Shared/Foley/Biped/TeddyBear/Wavs/Mimosa_Biped_TeddyBear_Walk_05}
 }
 {
  id=WavePlayer_34
  properties {soundWaveAssetPtr=/CosmeticCompanions/Assets/Audio_Shared/Foley/Biped/TeddyBear/Wavs/Mimosa_Biped_TeddyBear_Walk_04}
 }
 {
  id=WavePlayer_35
  properties {soundWaveAssetPtr=/CosmeticCompanions/Assets/Audio_Shared/Foley/Biped/TeddyBear/Wavs/Mimosa_Biped_TeddyBear_Walk_03}
 }
 {
  id=WavePlayer_36
  properties {soundWaveAssetPtr=/CosmeticCompanions/Assets/Audio_Shared/Foley/Biped/TeddyBear/Wavs/Mimosa_Biped_TeddyBear_Walk_02}
 }
 {
  id=WavePlayer_37
  properties {soundWaveAssetPtr=/CosmeticCompanions/Assets/Audio_Shared/Foley/Biped/TeddyBear/Wavs/Mimosa_Biped_TeddyBear_Walk_01}
 }
 {
  id=WavePlayer_38
  properties {soundWaveAssetPtr=&Mimosa_Biped_TeddyBear_Walk_04}
 }
 {
  id=WavePlayer_39
  properties {soundWaveAssetPtr=&Mimosa_Biped_TeddyBear_Walk_03}
 }
 {
  id=WavePlayer_40
  properties {soundWaveAssetPtr=&Mimosa_Biped_TeddyBear_Walk_02}
 }
 {
  id=WavePlayer_41
  properties {soundWaveAssetPtr=&Mimosa_Biped_TeddyBear_Walk_01}
 }
]
properties {
 attenuationSettings=/Game/Sounds/Attenuation/AS_Companion_Foley
 maxDistance=1200
 pitchMultiplier=0.9
 priority=0
 soundClassObject=&Fort_Creatures_Pets
 volumeMultiplier=0.5
 concurrencySet [/CosmeticCompanions/Assets/Audio_Shared/Foley/Biped/CONC_Mimosa_Biped_Footstep]
}
