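# Sound cue for the teddy bear biped companion's scratch foley in Cosmetic Companions.
# A Random node picks one of two scratch waves with equal weight.
# It plays at reduced volume in the creature pets sound class, using the shared companion foley attenuation.
# It uses the biped footstep concurrency set rather than a scratch-specific one.
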SoundCue /CosmeticCompanions/Assets/Audio_Shared/Foley/Biped/TeddyBear/SC_MImosa_Biped_TeddyBear_Scratch
root=Random_2
nodes [
 {id=Random_2, children [WavePlayer_0,WavePlayer_1], properties {weights [1,1]}}
 {
  id=WavePlayer_0
  properties {soundWaveAssetPtr=/CosmeticCompanions/Assets/Audio_Shared/Foley/Biped/TeddyBear/Wavs/Mimosa_Biped_TeddyBear_Scratch_01}
 }
 {
  id=WavePlayer_1
  properties {soundWaveAssetPtr=/CosmeticCompanions/Assets/Audio_Shared/Foley/Biped/TeddyBear/Wavs/Mimosa_Biped_TeddyBear_Scratch_02}
 }
]
properties {
 attenuationSettings=/Game/Sounds/Attenuation/AS_Companion_Foley
 maxDistance=1200
 soundClassObject=&Fort_Creatures_Pets
 volumeMultiplier=0.5
 concurrencySet [/CosmeticCompanions/Assets/Audio_Shared/Foley/Biped/CONC_Mimosa_Biped_Footstep]
}
